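# One-shot stereo MetaSound source that plays a Morphite weapon level-up-ready sound, routed to the gadgets and items sound class.
# On play, a random array get picks one of the LevelUpReady wave variants and triggers a stereo wave player.
# The player feeds both outputs directly and signals finished when done.
# A random-float gain multiply is present, but its output is unconnected, so it does not vary the audio.
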MetaSoundSource /BlastBerryWeapons/Audio/MSS_Morphite_LevelUpReady_Weapon
inputs [
 In Array {type=WaveAsset:Array}
 UE.Source.OnPlay {type=Trigger}
]
interfaces [UE.OutputFormat.Stereo:1.0, UE.Source.OneShot:1.0, UE.Source:1.1]
outputs [
 "UE.OutputFormat.Stereo.Audio:0" {type=Audio}
 "UE.OutputFormat.Stereo.Audio:1" {type=Audio}
 UE.Source.OneShot.OnFinished {type=Trigger}
]
pages [Default {
  input_defaults {
   UE.Source.OnPlay=null
   In Array [
    &/Morphite_LevelUpReady_05
    &/Morphite_LevelUpReady_04
    &/Morphite_LevelUpReady_03
    &/Morphite_LevelUpReady_02
    &/Morphite_LevelUpReady_01
   ]
  }
  nodes [
   UE.Multiply:Audio by Float(AdditionalOperands=UE.RandomFloat() {also {defaults {Max=1.2, Min=0.8}}}.Value)
   ArrayRandomGet_0 = Array.Random Get:WaveAsset:Array(In Array=^In Array, Next=^UE.Source.OnPlay)
   WavePlayer_0 = UE.Wave Player:Stereo(Play=@ArrayRandomGet_0.On Next, Wave Asset=@ArrayRandomGet_0.Value) {also {version [1,1]}}
  ]
  links [
   {dst="@UE.Source.OneShot.OnFinished", src=WavePlayer_0.On Finished}
   {dst="@UE.OutputFormat.Stereo.Audio:0", src=WavePlayer_0.Out Left}
   {dst="@UE.OutputFormat.Stereo.Audio:1", src=WavePlayer_0.Out Right}
  ]
 }]
properties {
 assetImportData=&self:AssetImportData
 numChannels=2
 outputFormat=Stereo
 sampleRate=48000
 soundClassObject=&Fort_Gadgets_Items
 volume=1.2
}
